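; Auto supply drop component that rifts driveable reboot vans into the match as the storm progresses, except in solo playlists.
; When the safe zone enters Holding within a configured phase range, it counts vans inside the zone and derives a spawn count from the maximum and ratio.
; An EQS query sized from the zone radius supplies shuffled candidate locations.
; Vans spawn at random picks with varied deploy height and rotation.
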(event EventBeginPlay
  (bind (_center _radius) (Utilities|Struct|BreakFortSafeZonePhaseInfo (SafeZone|GetPhaseInfo (Variables|Default|GetSafe%20Zone%20Indicator) (SafeZone|GetCurrentPhaseIndex (Variables|Default|GetSafe%20Zone%20Indicator)))))
  ; Cache Values
  (Class|FortAsyncActionCurrentPlaylistReady|PlaylistDataReadyAsync
    (:OnReady
      (Variables|Default|SetDeployHeightValue (ScalableFloat|GetValueatLevel (Variables|Default|GetDeployHeight)))
      (Variables|Default|SetPitchLimitValue (ScalableFloat|GetValueatLevel (Variables|Default|GetPitchLimit)))
      (Variables|Default|SetRollLimitValue (ScalableFloat|GetValueatLevel (Variables|Default|GetRollLimit)))
      (Variables|Default|SetHeightVariationValue (ScalableFloat|GetValueatLevel (Variables|Default|GetHeightVariation)))
      ; Check Safe Zone conditions for spawning (State, min and max index, etc)
      (bind _returnvalue (GameplayEventRouter|GetEventRouter "/Script/FortniteGame.FortGameState" (Components|GetOwner)))
      (bind (_asynctaskproxy _payload) (GameplayEventRouter|StartListeningToStatefulEvent _returnvalue "/Script/FortniteGame.FortSafeZoneStateUpdatedEvent")
        (:OnEventReceived
          (if (Utilities|Enum|Equal(Enum) (.CurrentState (Utilities|Struct|BreakFortSafeZoneStateUpdatedEvent _payload)) "Holding")
            (bind _output_get (Variables|Default|SetSafe%20Zone%20Indicator (Utilities|Struct|BreakFortSafeZoneStateUpdatedEvent _payload)))
            (if (>= (SafeZone|GetCurrentPhaseIndex _output_get) (Variables|Default|GetMinZoneReboot))
              (if (<= (SafeZone|GetCurrentPhaseIndex _output_get) (Variables|Default|GetMaxZoneReboot))
                ;[1 1 1] NOT IN SOLOS
                (if (not (Class|FortGameStateAthena|PlaylistHasTag (Fortnite|Athena|GameState|GetGameStateAthena) (Variables|Default|GetSoloTag)))
                  (Variables|Default|SetVansInSafeZone 0)
                  ;[1 1 1] How Many Vans Are Safe
                  (bind _outactors (Actor|GetAllActorsOfClass "/ImmunityCarriage/Vehicle/Valet_ImmunityCarriage_Vehicle.Valet_ImmunityCarriage_Vehicle_C"))
                  (for _array_element _outactors
                    (if (Variables|/Script/ImmunityCarriageRuntime.ImmunityCarriage_InteractionComponent|GetbIsInsideSafeZone (Variables|/ImmunityCarriage/Vehicle/Valet_ImmunityCarriage_Vehicle.Valet_ImmunityCarriage_Vehicle_C|GetSpawnMachineComponent _array_element))
                      (Math|Integer|IncrementInt (Variables|Default|GetVansInSafeZone))))
                  ; How Many Vans To Spawn
                  (bind _output_get_2 (Variables|Default|SetVansToSpawn (Math|Integer|Max(Integer) 0 (- (Math|Float|Truncate (ScalableFloat|GetValueasDouble (Variables|Default|GetMaxNumberToSpawn))) (Math|Float|Floor (Math|Float|SafeDivide (Math|Conversions|ToFloat(Integer) (Variables|Default|GetVansInSafeZone)) (ScalableFloat|GetValueasDouble (Variables|Default|GetSpawnRatio))))))))
                  (if (> _output_get_2 0)
                    ;[1 1 1] Query For Reboot Spawn Locations
                    (bind _returnvalue_2 (AI|EQS|RunEQSQuery (Variables|Default|GetEQSSafeZone) self "AllMatching"))
                    (AI|EQS|SetNamedParam _returnvalue_2 "OuterRadius" (* _radius 2.0))
                    (AI|EQS|AssignOnQueryFinished _returnvalue_2 (delegate Custom|OnQueryFinishedEvent))))))))))))

(event Custom|OnQueryFinishedEvent
  (params (QueryInstance :type EQS%20Query%20Instance%20Object%20Reference) (QueryStatus :type EEnvQueryStatus%20Enum :default "Processing"))
  (bind _allsafelocations (Variables|Default|GetAllSafeLocations))
  (bind _returnvalue (Utilities|Array|Length _allsafelocations))
  (bind _returnvalue_2 (- _returnvalue 1))
  (bind _returnvalue_3 (Math|Random|RandomIntegerinRange 0 _returnvalue_2))
  (bind _allsafelocations_2 (Variables|Default|GetAllSafeLocations))
  ;[1 1 1] Query For Reboot Spawn Locations
  (if (Utilities|Enum|Equal(Enum) QueryStatus "Success")
    (bind (_resultlocations _returnvalue_4) (AI|EQS|GetQueryResultsasLocations QueryInstance))
    (if _returnvalue_4
      (Utilities|FlowControl|Sequence
        (:then_0
          ;[1 1 1] Chose Locations
          (Utilities|Array|Clear (Variables|Default|GetChosenLocations))
          (bind _output_get (Variables|Default|SetAllSafeLocations _resultlocations))
          (Utilities|Array|Shuffle _output_get)
          (if (not (Utilities|Array|IsEmpty _output_get))
            (for _ (range (Variables|Default|GetVansToSpawn))
              (Utilities|Array|Add (Variables|Default|GetChosenLocations) (Utilities|Array|Get(aref) _allsafelocations_2 _returnvalue_3))
              (Utilities|Array|RemoveIndex _allsafelocations_2 _returnvalue_3))))
        (:then_1
          ;[1 1 1] Spawn Vans Normally
          (for _array_element (Variables|Default|GetChosenLocations)
            (CallFunction|SpawnVan _array_element)))))
    (else
      (Development|PrintString "Reboot Van Points Query Failed" true true "(R=1.000000,G=0.000000,B=0.059143,A=1.000000)"))))

(event Custom|SpawnVan
  (params (Location :type Vector))
  ; Spawn One Reboot Van
  (bind _output_get (Variables|Default|SetCurrentSafeZoneLocation Location))
  (bind (_output_get_x _output_get_y _output_get_z) (Math|Vector|BreakVector _output_get))
  (bind _returnvalue (Game|SpawnActorfromClass "/ImmunityCarriage/Vehicle/Valet_ImmunityCarriage_Vehicle.Valet_ImmunityCarriage_Vehicle_C" (Math|Transform|MakeTransform (Math|Vector|MakeVector _output_get_x _output_get_y (Utilities|Operators|Add _output_get_z (Variables|Default|GetDeployHeightValue) (Math|Random|RandomFloatinRange (* (Variables|Default|GetHeightVariationValue) -1.0) (Variables|Default|GetHeightVariationValue)))) (Utilities|MakeRandomVanRotation)) "AlwaysSpawn" "MultiplyWithRoot" 0 false 0))
  (Class|ValetImmunityCarriageVehicle|RiftRebootVanIntoWorld _returnvalue))
